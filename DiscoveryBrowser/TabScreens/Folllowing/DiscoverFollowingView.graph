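# Activatable Following tab screen of the Discover browser, listing followed content in a mega list.
# Construct and activation can point the Discover surface view model at the Following surface; FillOutListData loads providers into the list.
# SetLoadingState switches content, throbber and empty panels (Error shows empty), then refreshes focus to the list or back-to-discover button.
# Both back buttons navigate to ShowPlay; input-method changes also refresh focus.
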WidgetBlueprint /DiscoveryBrowser/TabScreens/Folllowing/DiscoverFollowingView
graphs [
 BP_GetDesiredFocusTarget {nodes [
   Get Desired Focus Target()
   > n1 = Branch(Condition=IsVisible(self=Get GP_Content() -> (GP_Content:Grid Panel Object Reference)))
   > Return Node!(
    execute
    ReturnValue:Widget Object Reference=Get ListView!(self:Fort Polymorphic List View Object Reference=Get WBP_Poly_MegaList() -> (WBP_Poly_MegaList:WBP Poly Mega List Object Reference)) -> (ListView:Common List View Object Reference)
   )
   Return Node(execute=@n1.else, ReturnValue=Get BackToDisco_Button() -> (BackToDisco_Button:WBP UIKit Button Loud Object Reference))
  ]}
 EventGraph {nodes [
   n1 = "On Clicked (BottomBarHack_Back)"!() -> (OutputDelegate, then, Button:Common Button Base Object Reference) {is=title K2Node_ComponentBoundEvent}
   NavigateToFeature!(execute=[@n1.then, @n2.then], Feature:EFortUIFeature Enum=ShowPlay, bClearPrimaryContentStack:bool=true) -> (then)
   n2 = "On Clicked (BackToDisco_Button)"!() -> (OutputDelegate, then, Button:Common Button Base Object Reference) {is=title K2Node_ComponentBoundEvent}
   OnInputMethodChanged!() -> (OutputDelegate, then, InputType:ECommonInputType Enum) {is=title K2Node_CustomEvent}
   > RequestRefreshFocus(self:Common Activatable Widget Object Reference)
   n3 = OnUserNavigatedMegaList!() -> (OutputDelegate, then, IsDirectionUp:bool=~, NewDesiredOuterIndex:int) {is=title K2Node_CustomEvent}
   > Branch(Condition=AND Boolean(A=NOT Boolean(A=@n3.IsDirectionUp), B="integer < integer"(A=@n3.NewDesiredOuterIndex, B=1)))
   > ScrollToStart(self:Scroll Box Object Reference=Get ScrollBox() -> (ScrollBox:Scroll Box Object Reference))
   Event PreConstruct()
   > UpdateVisual()
   Event Construct()
   n4 = Get FortDiscoverSurfaceViewModel() -> (FortDiscoverSurfaceViewModel:Discover Surface View Model Object Reference)
   > n5 = Branch(Condition=ShouldQueryOnConstruct!(self:Discover Surface View Model Object Reference=@n4) -> (ReturnValue:bool))
   > n6 = Set SurfaceName!(execute, SurfaceName:Name=FortniteDiscoverSurface_Following, self:Discover Surface View Model Object Reference=@n4) -> (then, Output_Get:Name=None)
   n7 = Create Event()
   Bind Event to On User Navigated Mega List(execute=[@n6.then, @n5.else], self:WBP Poly Mega List Object Reference=Get WBP_Poly_MegaList(), Delegate=@n7) {is=title K2Node_AddDelegate}
   n8 = Create Event()
   n9 = Get CommonInputSubsystem()
   > Bind Event to On Input Method Changed(self=@n9, Delegate=@n8)
   > OnInputMethodChanged!(execute, self, InputType:ECommonInputType Enum=GetCurrentInputType(self=Get CommonInputSubsystem())) -> (then) {is=title K2Node_CallFunction}
   Event On Activated()
   > SetScrollWidget!(execute, ToScroll:Widget Object Reference=Get ScrollBox()) -> (then)
   > n10 = Branch(Condition=ShouldQueryOnConstruct(self=Get FortDiscoverSurfaceViewModel()))
   Fire Event Main View Opened!(
    execute=[@n13.then, @n10.then, @n12.else]
    View:Widget Object Reference=Self-Reference()
    AdditionalInfo:Flow Analytics Additional Info Structure
    bIsOverlay:bool=false
   ) -> (then)
   n11 = Get FortDiscoverSurfaceViewModel()
   n12 = Branch(
    execute=@n10.else
    Condition="Equal (Name)"(A=Get SurfaceName!(self:Discover Surface View Model Object Reference=@n11) -> (SurfaceName:Name=None))
   )
   > n13 = Set SurfaceName(SurfaceName=FortniteDiscoverSurface_Following, self=@n11) -> (Output_Get=None)
   Event Destruct()
   > Unbind Event from On User Navigated Mega List(self:WBP Poly Mega List Object Reference=Get WBP_Poly_MegaList(), Delegate=@n7) {is=title K2Node_RemoveDelegate}
   > Unbind Event from On Input Method Changed(self=@n9, Delegate=@n8)
   Event On Deactivated()
  ]}
 FillOutListData {nodes [
   n1 = FillOutListData!() -> (then, Providers:Array of Discover Provider View Model Object References) {is=title K2Node_FunctionEntry}
   > ClearListItems(self=Get ListView(self=Get WBP_Poly_MegaList()))
   > Branch(Condition=Is Not Empty!(TargetArray:Array of Discover Provider View Model Object References=@n1) -> (ReturnValue:bool))
   > Set List Items!(execute, self:List View Object Reference=Get ListView(self=Get WBP_Poly_MegaList()), InListItems:Array of Object References=@n1) -> (then) {is=title K2Node_CallFunction}
  ]}
 SetLoadingState {nodes [
   n1 = SetLoadingState!() -> (then, LoadingState:EDiscoverLoadingState Enum=Unset) {is=title K2Node_FunctionEntry}
   > n2 = Switch on EDiscoverLoadingState!(execute, Selection:EDiscoverLoadingState Enum=@n1) -> (Unset:Exec, Loading:Exec, Loaded:Exec, Error:Exec) {is=title K2Node_SwitchEnum}
   SetVisibility(execute=[@n2.Error, @n2.Unset], self=Get GP_Empty() -> (GP_Empty:Grid Panel Object Reference), InVisibility=SelfHitTestInvisible)
   > n3 = SetVisibility(self=[Get GP_Loading() -> (GP_Loading:Grid Panel Object Reference), Get GP_Content()], InVisibility=Collapsed)
   RequestRefreshFocus(execute=[@n3.then, @n4.then])
   SetVisibility(execute=@n2.Loaded, self=Get GP_Content(), InVisibility=SelfHitTestInvisible)
   > n4 = SetVisibility(self=[Get GP_Empty(), Get GP_Loading()], InVisibility=Collapsed)
   SetVisibility(execute=@n2.Loading, self=Get GP_Loading(), InVisibility=SelfHitTestInvisible)
   > SetVisibility(self=[Get GP_Content(), Get GP_Empty()], InVisibility=Collapsed)
  ]}
 UpdateVisual {nodes [
   UpdateVisual() {is=title K2Node_FunctionEntry}
   > SetClipping!(
    execute
    self:Widget Object Reference=Get GridRoot!(self:WBP Poly Mega List Object Reference=Get WBP_Poly_MegaList()) -> (GridRoot:Grid Panel Object Reference)
    InClipping:EWidgetClipping Enum=ClipToBounds
   ) -> (then)
   n1 = Get WBP_Discovery_BlankTiles_RoundedCorners() -> (WBP_Discovery_BlankTiles_RoundedCorners:WBP Discovery Blank Tiles Rounded Corners Object Reference)
   > Set VerticalEntrySpacing!(
    execute
    VerticalEntrySpacing:float=16
    self:List View Object Reference=Get TileView_BlankTiles!(self:WBP Discovery Blank Tiles Rounded Corners Object Reference=@n1) -> (TileView_BlankTiles:Tile View Object Reference)
   ) -> (then, Output_Get:float)
   > SetEntryTileSize!(execute, self:WBP Discovery Blank Tiles Rounded Corners Object Reference=@n1, EntrySize:Vector 2D Structure=(X=256,Y=144)) -> (then)
  ]}
]
variables []
widgets {
 info {
  inheritedWidgetCount=0
  namedSlotCount=1
  parentClass=&CommonActivatableWidget
  rootWidgetClass=&GridPanel
  widgetCount=20
 }
 widgets [
  GridPanel_28 {widgetClassPath=&GridPanel, widget=GridPanel_28, properties {columnFill [1], rowFill [1]}}
  NebulousBackground {
   widgetClassPath=&Image
   bIsVariable=true
   parent=GridPanel_28
   widget=NebulousBackground
   properties {brush {
     bIsDynamicallyLoaded=false
     drawAs=Image
     imageType=NoImage
     mirroring=NoMirror
     resourceName=None
     resourceObject=/FortUILibrary/MaterialInstances/Background/Nebulous/MI_UI_Background_Nebulous
     tiling=NoTile
     imageSize=<32 32>
     margin=<ltrb 0 0 0 0>
     outlineSettings {
      bUseBrushTransparency=false
      roundingType=HalfHeightRadius
      width=0
      color {colorUseRule=UseColor_Specified, specifiedColor=<rgba 0 0 0 0>}
      cornerRadii=<0 0 0 0>
     }
     tintColor {colorUseRule=UseColor_Specified, specifiedColor=<rgba 1 1 1 1>}
     uVRegion {bIsValid=false, max=<0 0>, min=<0 0>}
    }}
   slot {class=GridSlot}
  }
  Image_SoftEdgeGradient_MegaList {
   widgetClassPath=&Image
   bIsVariable=true
   parent=GridPanel_28
   widget=Image_SoftEdgeGradient_MegaList
   properties {
    visibility=SelfHitTestInvisible
    brush {
     bIsDynamicallyLoaded=false
     drawAs=Image
     imageType=FullColor
     mirroring=NoMirror
     resourceName=None
     resourceObject=/DiscoveryBrowser/TabScreens/Search/MaterialInstances/MI_BackgroundSoftEdge_Nebulous
     tiling=NoTile
     imageSize=<32 150>
     margin=<ltrb 0 0 0 0>
     outlineSettings {
      bUseBrushTransparency=false
      roundingType=HalfHeightRadius
      width=0
      color {colorUseRule=UseColor_Specified, specifiedColor=<rgba 0 0 0 0>}
      cornerRadii=<0 0 0 0>
     }
     tintColor {colorUseRule=UseColor_Specified, specifiedColor=<rgba 1 1 1 1>}
     uVRegion {bIsValid=false, max=<0 0>, min=<0 0>}
    }
   }
   slot {class=GridSlot, properties {layer=11, verticalAlignment=VAlign_Top}}
  }
  CaptureForPostBufferUpdate {
   widgetClassPath=&WBP_CaptureForPostBufferUpdate
   bIsVariable=true
   parent=GridPanel_28
   widget=CaptureForPostBufferUpdate
   slot {class=GridSlot, properties {layer=1}}
  }
  SafeZone {
   widgetClassPath=&SafeZone
   bIsVariable=true
   parent=GridPanel_28
   widget=SafeZone
   properties {padBottom=false, padLeft=false, padRight=false, padTop=false}
   slot {class=GridSlot, properties {layer=10, padding=<ltrb 0 160 0 0>}}
  }
  ScrollBox {
   widgetClassPath=&ScrollBox
   bIsVariable=true
   parent=SafeZone
   widget=ScrollBox
   properties {
    allowOverscroll=false
    bAllowRightClickDragScrolling=false
    clipping=Inherit
    navigationScrollPadding=120
    scrollBarVisibility=Collapsed
    scrollWhenFocusChanges=InstantScroll
    wheelScrollMultiplier=3
    scrollbarPadding=<ltrb 0 0 0 0>
    scrollbarThickness=<0 0>
    widgetBarStyle {
     thickness=0
     draggedThumbImage {
      bIsDynamicallyLoaded=false
      drawAs=NoDrawType
      imageType=FullColor
      mirroring=NoMirror
      resourceName=../../../Engine/Content/Slate/Common/Scrollbar_Thumb.png
      resourceObject=null
      tiling=NoTile
      imageSize=<0 0>
      margin=<ltrb 0.25 0.25 0.25 0.25>
      outlineSettings {
       bUseBrushTransparency=false
       roundingType=HalfHeightRadius
       width=0
       color {colorUseRule=UseColor_Specified, specifiedColor=<rgba 0 0 0 0>}
       cornerRadii=<0 0 0 0>
      }
      tintColor {colorUseRule=UseColor_Specified, specifiedColor=<rgba 1 1 1 1>}
      uVRegion {bIsValid=false, max=<0 0>, min=<0 0>}
     }
     horizontalBackgroundImage {
      bIsDynamicallyLoaded=false
      drawAs=NoDrawType
      imageType=NoImage
      mirroring=NoMirror
      resourceName=None
      resourceObject=null
      tiling=NoTile
      imageSize=<0 0>
      margin=<ltrb 0 0 0 0>
      outlineSettings {
       bUseBrushTransparency=false
       roundingType=HalfHeightRadius
       width=0
       color {colorUseRule=UseColor_Specified, specifiedColor=<rgba 0 0 0 0>}
       cornerRadii=<0 0 0 0>
      }
      tintColor {colorUseRule=UseColor_Specified, specifiedColor=<rgba 1 1 1 1>}
      uVRegion {bIsValid=false, max=<0 0>, min=<0 0>}
     }
     horizontalBottomSlotImage {
      bIsDynamicallyLoaded=false
      drawAs=NoDrawType
      imageType=FullColor
      mirroring=NoMirror
      resourceName=../../../Engine/Content/Slate/Common/Scrollbar_Background_Horizontal.png
      resourceObject=null
      tiling=NoTile
      imageSize=<8 8>
      margin=<ltrb 0 0 0 0>
      outlineSettings {
       bUseBrushTransparency=false
       roundingType=HalfHeightRadius
       width=0
       color {colorUseRule=UseColor_Specified, specifiedColor=<rgba 0 0 0 0>}
       cornerRadii=<0 0 0 0>
      }
      tintColor {colorUseRule=UseColor_Specified, specifiedColor=<rgba 1 1 1 1>}
      uVRegion {bIsValid=false, max=<0 0>, min=<0 0>}
     }
     horizontalTopSlotImage {
      bIsDynamicallyLoaded=false
      drawAs=NoDrawType
      imageType=FullColor
      mirroring=NoMirror
      resourceName=../../../Engine/Content/Slate/Common/Scrollbar_Background_Horizontal.png
      resourceObject=null
      tiling=NoTile
      imageSize=<8 8>
      margin=<ltrb 0 0 0 0>
      outlineSettings {
       bUseBrushTransparency=false
       roundingType=HalfHeightRadius
       width=0
       color {colorUseRule=UseColor_Specified, specifiedColor=<rgba 0 0 0 0>}
       cornerRadii=<0 0 0 0>
      }
      tintColor {colorUseRule=UseColor_Specified, specifiedColor=<rgba 1 1 1 1>}
      uVRegion {bIsValid=false, max=<0 0>, min=<0 0>}
     }
     hoveredThumbImage {
      bIsDynamicallyLoaded=false
      drawAs=NoDrawType
      imageType=FullColor
      mirroring=NoMirror
      resourceName=../../../Engine/Content/Slate/Common/Scrollbar_Thumb.png
      resourceObject=null
      tiling=NoTile
      imageSize=<0 0>
      margin=<ltrb 0.25 0.25 0.25 0.25>
      outlineSettings {
       bUseBrushTransparency=false
       roundingType=HalfHeightRadius
       width=0
       color {colorUseRule=UseColor_Specified, specifiedColor=<rgba 0 0 0 0>}
       cornerRadii=<0 0 0 0>
      }
      tintColor {colorUseRule=UseColor_Specified, specifiedColor=<rgba 1 1 1 1>}
      uVRegion {bIsValid=false, max=<0 0>, min=<0 0>}
     }
     normalThumbImage {
      bIsDynamicallyLoaded=false
      drawAs=NoDrawType
      imageType=FullColor
      mirroring=NoMirror
      resourceName=../../../Engine/Content/Slate/Common/Scrollbar_Thumb.png
      resourceObject=null
      tiling=NoTile
      imageSize=<0 0>
      margin=<ltrb 0.25 0.25 0.25 0.25>
      outlineSettings {
       bUseBrushTransparency=false
       roundingType=HalfHeightRadius
       width=0
       color {colorUseRule=UseColor_Specified, specifiedColor=<rgba 0 0 0 0>}
       cornerRadii=<0 0 0 0>
      }
      tintColor {colorUseRule=UseColor_Specified, specifiedColor=<rgba 1 1 1 1>}
      uVRegion {bIsValid=false, max=<0 0>, min=<0 0>}
     }
     verticalBackgroundImage {
      bIsDynamicallyLoaded=false
      drawAs=NoDrawType
      imageType=NoImage
      mirroring=NoMirror
      resourceName=None
      resourceObject=null
      tiling=NoTile
      imageSize=<0 0>
      margin=<ltrb 0 0 0 0>
      outlineSettings {
       bUseBrushTransparency=false
       roundingType=HalfHeightRadius
       width=0
       color {colorUseRule=UseColor_Specified, specifiedColor=<rgba 0 0 0 0>}
       cornerRadii=<0 0 0 0>
      }
      tintColor {colorUseRule=UseColor_Specified, specifiedColor=<rgba 1 1 1 1>}
      uVRegion {bIsValid=false, max=<0 0>, min=<0 0>}
     }
     verticalBottomSlotImage {
      bIsDynamicallyLoaded=false
      drawAs=NoDrawType
      imageType=FullColor
      mirroring=NoMirror
      resourceName=../../../Engine/Content/Slate/Common/Scrollbar_Background_Vertical.png
      resourceObject=null
      tiling=NoTile
      imageSize=<8 8>
      margin=<ltrb 0 0 0 0>
      outlineSettings {
       bUseBrushTransparency=false
       roundingType=HalfHeightRadius
       width=0
       color {colorUseRule=UseColor_Specified, specifiedColor=<rgba 0 0 0 0>}
       cornerRadii=<0 0 0 0>
      }
      tintColor {colorUseRule=UseColor_Specified, specifiedColor=<rgba 1 1 1 1>}
      uVRegion {bIsValid=false, max=<0 0>, min=<0 0>}
     }
     verticalTopSlotImage {
      bIsDynamicallyLoaded=false
      drawAs=NoDrawType
      imageType=FullColor
      mirroring=NoMirror
      resourceName=../../../Engine/Content/Slate/Common/Scrollbar_Background_Vertical.png
      resourceObject=null
      tiling=NoTile
      imageSize=<8 8>
      margin=<ltrb 0 0 0 0>
      outlineSettings {
       bUseBrushTransparency=false
       roundingType=HalfHeightRadius
       width=0
       color {colorUseRule=UseColor_Specified, specifiedColor=<rgba 0 0 0 0>}
       cornerRadii=<0 0 0 0>
      }
      tintColor {colorUseRule=UseColor_Specified, specifiedColor=<rgba 1 1 1 1>}
      uVRegion {bIsValid=false, max=<0 0>, min=<0 0>}
     }
    }
    widgetStyle {
     barThickness=0
     bottomShadowBrush {
      bIsDynamicallyLoaded=false
      drawAs=NoDrawType
      imageType=FullColor
      mirroring=NoMirror
      resourceName=../../../Engine/Content/Slate/Common/ScrollBoxShadowBottom.png
      resourceObject=null
      tiling=NoTile
      imageSize=<16 8>
      margin=<ltrb 0.5 0 0.5 1>
      outlineSettings {
       bUseBrushTransparency=false
       roundingType=HalfHeightRadius
       width=0
       color {colorUseRule=UseColor_Specified, specifiedColor=<rgba 0 0 0 0>}
       cornerRadii=<0 0 0 0>
      }
      tintColor {colorUseRule=UseColor_Specified, specifiedColor=<rgba 1 1 1 1>}
      uVRegion {bIsValid=false, max=<0 0>, min=<0 0>}
     }
     horizontalScrolledContentPadding=<ltrb 0 0 0 0>
     leftShadowBrush {
      bIsDynamicallyLoaded=false
      drawAs=NoDrawType
      imageType=FullColor
      mirroring=NoMirror
      resourceName=../../../Engine/Content/Slate/Common/ScrollBoxShadowLeft.png
      resourceObject=null
      tiling=NoTile
      imageSize=<8 16>
      margin=<ltrb 1 0.5 0 0.5>
      outlineSettings {
       bUseBrushTransparency=false
       roundingType=HalfHeightRadius
       width=0
       color {colorUseRule=UseColor_Specified, specifiedColor=<rgba 0 0 0 0>}
       cornerRadii=<0 0 0 0>
      }
      tintColor {colorUseRule=UseColor_Specified, specifiedColor=<rgba 1 1 1 1>}
      uVRegion {bIsValid=false, max=<0 0>, min=<0 0>}
     }
     rightShadowBrush {
      bIsDynamicallyLoaded=false
      drawAs=NoDrawType
      imageType=FullColor
      mirroring=NoMirror
      resourceName=../../../Engine/Content/Slate/Common/ScrollBoxShadowRight.png
      resourceObject=null
      tiling=NoTile
      imageSize=<8 16>
      margin=<ltrb 0 0.5 1 0.5>
      outlineSettings {
       bUseBrushTransparency=false
       roundingType=HalfHeightRadius
       width=0
       color {colorUseRule=UseColor_Specified, specifiedColor=<rgba 0 0 0 0>}
       cornerRadii=<0 0 0 0>
      }
      tintColor {colorUseRule=UseColor_Specified, specifiedColor=<rgba 1 1 1 1>}
      uVRegion {bIsValid=false, max=<0 0>, min=<0 0>}
     }
     topShadowBrush {
      bIsDynamicallyLoaded=false
      drawAs=NoDrawType
      imageType=FullColor
      mirroring=NoMirror
      resourceName=../../../Engine/Content/Slate/Common/ScrollBoxShadowTop.png
      resourceObject=null
      tiling=NoTile
      imageSize=<16 8>
      margin=<ltrb 0.5 1 0.5 0>
      outlineSettings {
       bUseBrushTransparency=false
       roundingType=HalfHeightRadius
       width=0
       color {colorUseRule=UseColor_Specified, specifiedColor=<rgba 0 0 0 0>}
       cornerRadii=<0 0 0 0>
      }
      tintColor {colorUseRule=UseColor_Specified, specifiedColor=<rgba 1 1 1 1>}
      uVRegion {bIsValid=false, max=<0 0>, min=<0 0>}
     }
     verticalScrolledContentPadding=<ltrb 0 0 0 0>
    }
   }
   slot {class=SafeZoneSlot}
  }
  GP_Content {
   widgetClassPath=&GridPanel
   bIsVariable=true
   parent=ScrollBox
   widget=GP_Content
   properties {columnFill [1], rowFill [1]}
   slot {class=ScrollBoxSlot}
  }
  WBP_Poly_MegaList {
   widgetClassPath=/DiscoveryBrowser/WBP_Poly_MegaList
   bIsVariable=true
   parent=GP_Content
   widget=WBP_Poly_MegaList
   slot {class=GridSlot, properties {row=1}}
  }
  CommonTextBlock_44 {
   widgetClassPath=&CommonTextBlock
   parent=GP_Content
   widget=CommonTextBlock_44
   properties {
    bIsScrollingEnabled=false
    text=Following
    font {
     bForceMonospaced=false
     bMaterialIsStencil=false
     fontMaterial=null
     fontObject=&NotoSans
     letterSpacing=0
     monospacedWidth=1
     size=24
     skewAmount=0
     typefaceFontName=NotoSans-Regular
     outlineSettings {
      bApplyOutlineToDropShadows=false
      bMiteredCorners=false
      bSeparateFillAlpha=false
      outlineMaterial=null
      outlineSize=0
      outlineColor=<rgba 0 0 0 1>
     }
    }
    shadowOffset=<0 0>
   }
   slot {class=GridSlot, properties {padding=<ltrb 80 0 0 64>}}
  }
  GP_Loading {
   widgetClassPath=&GridPanel
   bIsVariable=true
   parent=ScrollBox
   widget=GP_Loading
   properties {columnFill [1], rowFill [1]}
   slot {
    class=ScrollBoxSlot
    properties {horizontalAlignment=HAlign_Center, verticalAlignment=VAlign_Center, padding=<ltrb 0 0 0 80>, size {sizeRule=Fill, value=1}}
   }
  }
  WBP_UIKit_Throbber {
   widgetClassPath=/FortUILibrary/UIKit/Throbber/WBP_UIKit_Throbber
   bIsVariable=true
   parent=GP_Loading
   widget=WBP_UIKit_Throbber
   properties {displayBackground=false, throbberSize=Large}
   slot {class=GridSlot, properties {horizontalAlignment=HAlign_Center, verticalAlignment=VAlign_Center}}
  }
  GP_Empty {
   widgetClassPath=&GridPanel
   bIsVariable=true
   parent=ScrollBox
   widget=GP_Empty
   properties {columnFill [1], rowFill [1]}
   slot {class=ScrollBoxSlot}
  }
  WBP_Discovery_BlankTiles_RoundedCorners {
   widgetClassPath=/DiscoveryBrowser/WBP_Discovery_BlankTiles_RoundedCorners
   bIsVariable=true
   parent=GP_Empty
   widget=WBP_Discovery_BlankTiles_RoundedCorners
   properties {entries Count=24, flavorSubtitleOverride=Favorite islands and follow creators from Discover, flavorTitleOverride="Nothing to see yet!"}
   slot {class=GridSlot, properties {padding=<ltrb 80 0 80 0>}}
  }
  HorizontalBox_3 {
   widgetClassPath=&HorizontalBox
   namedSlotHost=WBP_Discovery_BlankTiles_RoundedCorners
   widget=HorizontalBox_3
   properties {bOverride_Cursor=true}
  }
  BackToDisco_Button {
   widgetClassPath=/FortUILibrary/UIKit/Buttons/WBP_UIKit_Button_Loud
   bIsVariable=true
   parent=HorizontalBox_3
   widget=BackToDisco_Button
   properties {
    navigation=&self:WidgetTree.BackToDisco_Button.WidgetNavigation_0
    text=back to discover
   }
   slot {
    class=HorizontalBoxSlot
    properties {horizontalAlignment=HAlign_Center, verticalAlignment=VAlign_Bottom, padding=<ltrb 0 32 0 0>, size {sizeRule=Fill, value=1}}
   }
  }
  SZBottomButtons {
   widgetClassPath=&SafeZone
   parent=GridPanel_28
   widget=SZBottomButtons
   properties {padTop=false}
   slot {class=GridSlot, properties {horizontalAlignment=HAlign_Right, layer=20, verticalAlignment=VAlign_Bottom, padding=<ltrb 0 0 80 32>}}
  }
  Overlay_1 {widgetClassPath=&Overlay, parent=SZBottomButtons, widget=Overlay_1, slot {class=SafeZoneSlot}}
  WBP_CaptureForPostBufferUpdate {
   widgetClassPath=&WBP_CaptureForPostBufferUpdate
   bIsVariable=true
   parent=Overlay_1
   slot {class=OverlaySlot, properties {horizontalAlignment=HAlign_Fill, verticalAlignment=VAlign_Fill}}
  }
  HBBottomButtons {
   widgetClassPath=&HorizontalBox
   parent=Overlay_1
   widget=HBBottomButtons
   slot {class=OverlaySlot, properties {horizontalAlignment=HAlign_Fill, verticalAlignment=VAlign_Fill}}
  }
  BottomBarHack_Back {
   widgetClassPath=/Game/UI/Frontend/BottomBar/WBP_BottomBarDecoyButton
   bIsVariable=true
   parent=HBBottomButtons
   widget=BottomBarHack_Back
   properties {minWidth=120, triggeringInputAction {dataTable=&InputActionDataTable, rowName=Back_SearchResult}}
   slot {class=HorizontalBoxSlot}
  }
 ]
}
